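# Parent material for the GrindRex character body that adds reactive color-cycling shoe glow and blinking emissive effects.
# It takes MF_CharacterParent_2023_PreFX output and builds a dimmed state plus an active state, using a tech art mask, position-based gradient, four-step color switches and HueShift.
# Blueprint-named running, front-end and timestamp parameters, with IsReactive, drive the blend between states before MF_CharacterParent_2023_PostFX.
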Material /BRCosmetics/Characters/Player/Male/Medium/Bodies/M_MED_GrindRex/Materials/M_GrindRex_Body_Parent
graph {
 expressions {
  MaterialExpressionAdd_0 {
   class=MaterialExpressionAdd
   inputs [
    A {from=MaterialExpressionMultiply_3}
    B {from=""}
   ]
  }
  MaterialExpressionAdd_1 {
   class=MaterialExpressionAdd
   inputs [
    A {from=MaterialExpressionNamedRerouteUsage_8}
    B {from=MaterialExpressionNamedRerouteUsage_5}
   ]
  }
  MaterialExpressionAdd_2 {
   class=MaterialExpressionAdd
   inputs [
    A {from=MaterialExpressionNamedRerouteUsage_4}
    B {from=""}
   ]
  }
  MaterialExpressionAdd_3 {
   class=MaterialExpressionAdd
   inputs [
    A {from=MaterialExpressionTime_0}
    B {from=MaterialExpressionConstant_1}
   ]
   properties {constB=-2}
  }
  MaterialExpressionAdd_4 {
   class=MaterialExpressionAdd
   inputs [
    A {from=MaterialExpressionStep_2}
    B {from=MaterialExpressionMultiply_18}
   ]
  }
  MaterialExpressionAdd_5 {
   class=MaterialExpressionAdd
   inputs [
    A {from=MaterialExpressionTime_0}
    B {from=MaterialExpressionAdd_6}
   ]
   properties {constB=-2}
  }
  MaterialExpressionAdd_6 {
   class=MaterialExpressionAdd
   inputs [
    A {from=MaterialExpressionConstant_1}
    B {from=""}
   ]
   properties {constB=0.5}
  }
  MaterialExpressionBlendMaterialAttributes_0 {
   class=MaterialExpressionBlendMaterialAttributes
   inputs [
    A {from=MaterialExpressionSetMaterialAttributes_2, output=""}
    B {from=MaterialExpressionSetMaterialAttributes_0, output=""}
    Alpha {from=MaterialExpressionLinearInterpolate_2, output=""}
   ]
   properties {}
  }
  MaterialExpressionComponentMask_0 {
   class=MaterialExpressionComponentMask
   inputs [None {from=MaterialExpressionMaterialFunctionCall_0, output=Tech Art Mask}]
   properties {b=true}
  }
  MaterialExpressionComponentMask_1 {
   class=MaterialExpressionComponentMask
   inputs [None {from=MaterialExpressionMaterialFunctionCall_0, output=Pre skinned Local Position}]
   properties {b=true}
  }
  MaterialExpressionConstant3Vector_0 {class=MaterialExpressionConstant3Vector, properties {constant=<rgba 2.5 0.7822 0 1>}}
  MaterialExpressionConstant3Vector_1 {class=MaterialExpressionConstant3Vector, properties {constant=<rgba 2 1.3345 0 1>}}
  MaterialExpressionConstant3Vector_2 {class=MaterialExpressionConstant3Vector, properties {constant=<rgba 0 1.4402 2 1>}}
  MaterialExpressionConstant3Vector_3 {class=MaterialExpressionConstant3Vector, properties {constant=<rgba 2 0.12 0.1269 1>}}
  MaterialExpressionConstantBiasScale_0 {class=MaterialExpressionConstantBiasScale, inputs [None {from=MaterialExpressionSine_0}]}
  MaterialExpressionConstant_0 {class=MaterialExpressionConstant}
  MaterialExpressionConstant_1 {class=MaterialExpressionConstant, properties {r=-4}}
  MaterialExpressionConstant_2 {class=MaterialExpressionConstant, properties {r=4.05}}
  MaterialExpressionConstant_3 {class=MaterialExpressionConstant}
  MaterialExpressionConstant_4 {class=MaterialExpressionConstant, properties {r=0.6}}
  MaterialExpressionConstant_5 {class=MaterialExpressionConstant, properties {r=0.55}}
  MaterialExpressionConstant_6 {class=MaterialExpressionConstant, properties {r=0.45}}
  MaterialExpressionConstant_7 {class=MaterialExpressionConstant, properties {r=0.025}}
  MaterialExpressionDesaturation_2 {
   class=MaterialExpressionDesaturation
   inputs [
    None {from=MaterialExpressionNamedRerouteDeclaration_7, output=""}
    Fraction {from=MaterialExpressionMultiply_16, output=""}
   ]
   properties {}
  }
  MaterialExpressionFloor_0 {class=MaterialExpressionFloor, inputs [None {from=MaterialExpressionMultiply_9, output=""}], properties {}}
  MaterialExpressionFloor_1 {class=MaterialExpressionFloor, inputs [None {from=MaterialExpressionOneMinus_2, output=""}], properties {}}
  MaterialExpressionFrac_0 {class=MaterialExpressionFrac, inputs [None {from=MaterialExpressionMultiply_10, output=""}], properties {}}
  MaterialExpressionGetMaterialAttributes_0 {
   class=MaterialExpressionGetMaterialAttributes
   inputs [None {from=MaterialExpressionMaterialFunctionCall_0, output=Result}]
   properties {attributeGetTypes [#1, #2]}
  }
  MaterialExpressionLinearInterpolate_0 {
   class=MaterialExpressionLinearInterpolate
   inputs [
    A {from=MaterialExpressionNamedRerouteUsage_10}
    B {from=MaterialExpressionConstant_0}
    Alpha {from=MaterialExpressionNamedRerouteUsage_11}
   ]
  }
  MaterialExpressionLinearInterpolate_2 {
   class=MaterialExpressionLinearInterpolate
   inputs [
    A {from=MaterialExpressionMultiply_8}
    B {from=MaterialExpressionNamedRerouteUsage_12}
    Alpha {from=MaterialExpressionReroute_0}
   ]
  }
  MaterialExpressionMaterialFunctionCall_0 {
   class=MaterialExpressionMaterialFunctionCall
   inputs [
    Overide Texture Set UVs {from=""}
    Override Tech Art Mask UVs {from=""}
    Override Basic Emissive UVs {from=""}
   ]
   properties {materialFunction=&MF_CharacterParent_2023_PreFX}
  }
  MaterialExpressionMaterialFunctionCall_1 {
   class=MaterialExpressionMaterialFunctionCall
   inputs [
    Material {from=MaterialExpressionBlendMaterialAttributes_0}
    M {from=MaterialExpressionMaterialFunctionCall_0, output=M}
    Pre Skinned Local Position {from=MaterialExpressionMaterialFunctionCall_0, output=M}
    Distance Blend {from=MaterialExpressionMaterialFunctionCall_0, output=M}
    Fresnel {from=MaterialExpressionMaterialFunctionCall_0, output=M}
   ]
   properties {materialFunction=&MF_CharacterParent_2023_PostFX}
  }
  MaterialExpressionMaterialFunctionCall_3 {
   class=MaterialExpressionMaterialFunctionCall
   inputs [
    Hue Shift Percentage {from=MaterialExpressionSwitch_1}
    Texture {from=MaterialExpressionNamedRerouteUsage_6}
   ]
   properties {materialFunction=&HueShift}
  }
  MaterialExpressionMax_0 {
   class=MaterialExpressionMax
   inputs [
    A {from=MaterialExpressionOneMinus_0}
    B {from=""}
   ]
   properties {constB=0.5}
  }
  MaterialExpressionMultiply_0 {
   class=MaterialExpressionMultiply
   inputs [
    A {from=MaterialExpressionDesaturation_2}
    B {from=MaterialExpressionMax_0}
   ]
  }
  MaterialExpressionMultiply_1 {
   class=MaterialExpressionMultiply
   inputs [
    A {from=MaterialExpressionSphereMask_0}
    B {from=""}
   ]
   properties {constB=3}
  }
  MaterialExpressionMultiply_10 {
   class=MaterialExpressionMultiply
   inputs [
    A {from=MaterialExpressionMultiply_7}
    B {from=""}
   ]
   properties {constB=0.5}
  }
  MaterialExpressionMultiply_12 {
   class=MaterialExpressionMultiply
   inputs [
    A {from=MaterialExpressionNamedRerouteUsage_3}
    B {from=MaterialExpressionSwitch_0}
   ]
  }
  MaterialExpressionMultiply_13 {
   class=MaterialExpressionMultiply
   inputs [
    A {from=MaterialExpressionMultiply_5}
    B {from=""}
   ]
   properties {constB=0.25}
  }
  MaterialExpressionMultiply_15 {
   class=MaterialExpressionMultiply
   inputs [
    A {from=""}
    B {from=MaterialExpressionGetMaterialAttributes_0, output=Emissive Color}
   ]
   properties {constA=0.2}
  }
  MaterialExpressionMultiply_16 {
   class=MaterialExpressionMultiply
   inputs [
    A {from=MaterialExpressionNamedRerouteUsage_9}
    B {from=""}
   ]
   properties {constB=0.3}
  }
  MaterialExpressionMultiply_18 {
   class=MaterialExpressionMultiply
   inputs [
    A {from=MaterialExpressionNamedRerouteUsage_14}
    B {from=MaterialExpressionScalarParameter_3}
   ]
  }
  MaterialExpressionMultiply_19 {
   class=MaterialExpressionMultiply
   inputs [
    A {from=MaterialExpressionSubtract_2}
    B {from=""}
   ]
   properties {constB=10}
  }
  MaterialExpressionMultiply_2 {
   class=MaterialExpressionMultiply
   inputs [
    A {from=MaterialExpressionMaterialFunctionCall_3, output=Result}
    B {from=MaterialExpressionAdd_0}
   ]
  }
  MaterialExpressionMultiply_3 {
   class=MaterialExpressionMultiply
   inputs [
    A {from=MaterialExpressionConstantBiasScale_0}
    B {from=""}
   ]
   properties {constB=5}
  }
  MaterialExpressionMultiply_4 {
   class=MaterialExpressionMultiply
   inputs [
    A {from=MaterialExpressionTime_1}
    B {from=""}
   ]
   properties {constB=2}
  }
  MaterialExpressionMultiply_5 {
   class=MaterialExpressionMultiply
   inputs [
    A {from=MaterialExpressionAdd_2}
    B {from=MaterialExpressionMultiply_12}
   ]
  }
  MaterialExpressionMultiply_6 {
   class=MaterialExpressionMultiply
   inputs [
    A {from=MaterialExpressionOneMinus_1}
    B {from=MaterialExpressionStep_1}
   ]
  }
  MaterialExpressionMultiply_7 {
   class=MaterialExpressionMultiply
   inputs [
    A {from=MaterialExpressionNamedRerouteUsage_2}
    B {from=MaterialExpressionTime_2}
   ]
  }
  MaterialExpressionMultiply_8 {
   class=MaterialExpressionMultiply
   inputs [
    A {from=MaterialExpressionSmoothStep_0}
    B {from=MaterialExpressionScalarParameter_2}
   ]
  }
  MaterialExpressionMultiply_9 {
   class=MaterialExpressionMultiply
   inputs [
    A {from=MaterialExpressionFrac_0}
    B {from=""}
   ]
   properties {constB=4}
  }
  MaterialExpressionNamedRerouteDeclaration_0 {
   class=MaterialExpressionNamedRerouteDeclaration
   inputs [None {from=MaterialExpressionFloor_0}]
   properties {name=ShoeColorSwitch_Timing, nodeColor=<rgba 1 0 0.3516 1>}
  }
  MaterialExpressionNamedRerouteDeclaration_1 {
   class=MaterialExpressionNamedRerouteDeclaration
   inputs [None {from=MaterialExpressionSaturate_2}]
   properties {name=AlternateColors, nodeColor=<rgba 1 0 0.3516 1>}
  }
  MaterialExpressionNamedRerouteDeclaration_10 {
   class=MaterialExpressionNamedRerouteDeclaration
   inputs [None {from=MaterialExpressionFloor_1}]
   properties {name=FrontEnd_AlternateColors, nodeColor=<rgba 0.4297 0 1 1>}
  }
  MaterialExpressionNamedRerouteDeclaration_2 {
   class=MaterialExpressionNamedRerouteDeclaration
   inputs [None {from=MaterialExpressionComponentMask_0}]
   properties {name=ShoeMask, nodeColor=<rgba 0 1 0.6484 1>}
  }
  MaterialExpressionNamedRerouteDeclaration_3 {
   class=MaterialExpressionNamedRerouteDeclaration
   inputs [None {from=MaterialExpressionMultiply_1}]
   properties {name=GlowGradient, nodeColor=<rgba 1 0.3281 0 1>}
  }
  MaterialExpressionNamedRerouteDeclaration_4 {
   class=MaterialExpressionNamedRerouteDeclaration
   inputs [None {from=MaterialExpressionMultiply_13}]
   properties {name=ShoeGlowingColors, nodeColor=<rgba 0.0078 0 1 1>}
  }
  MaterialExpressionNamedRerouteDeclaration_5 {
   class=MaterialExpressionNamedRerouteDeclaration
   inputs [None {from=MaterialExpressionGetMaterialAttributes_0, output=Emissive Color}]
   properties {name=OriginalEmissive, nodeColor=<rgba 0.3125 1 0 1>}
  }
  MaterialExpressionNamedRerouteDeclaration_6 {
   class=MaterialExpressionNamedRerouteDeclaration
   inputs [None {from=MaterialExpressionMultiply_2}]
   properties {name=Blinking Emissive, nodeColor=<rgba 1 0 0.6328 1>}
  }
  MaterialExpressionNamedRerouteDeclaration_7 {
   class=MaterialExpressionNamedRerouteDeclaration
   inputs [None {from=MaterialExpressionGetMaterialAttributes_0, output=Base Color}]
   properties {name=OriginalBaseColor, nodeColor=<rgba 0.8047 0 1 1>}
  }
  MaterialExpressionNamedRerouteDeclaration_8 {
   class=MaterialExpressionNamedRerouteDeclaration
   inputs [None {from=MaterialExpressionMultiply_6}]
   properties {name=FrontEndTiming, nodeColor=<rgba 0 0.5078 1 1>}
  }
  MaterialExpressionNamedRerouteUsage_0 {class=MaterialExpressionNamedRerouteUsage}
  MaterialExpressionNamedRerouteUsage_10 {class=MaterialExpressionNamedRerouteUsage}
  MaterialExpressionNamedRerouteUsage_11 {class=MaterialExpressionNamedRerouteUsage}
  MaterialExpressionNamedRerouteUsage_12 {class=MaterialExpressionNamedRerouteUsage}
  MaterialExpressionNamedRerouteUsage_14 {class=MaterialExpressionNamedRerouteUsage}
  MaterialExpressionNamedRerouteUsage_2 {class=MaterialExpressionNamedRerouteUsage}
  MaterialExpressionNamedRerouteUsage_3 {class=MaterialExpressionNamedRerouteUsage}
  MaterialExpressionNamedRerouteUsage_4 {class=MaterialExpressionNamedRerouteUsage}
  MaterialExpressionNamedRerouteUsage_5 {class=MaterialExpressionNamedRerouteUsage}
  MaterialExpressionNamedRerouteUsage_6 {class=MaterialExpressionNamedRerouteUsage}
  MaterialExpressionNamedRerouteUsage_7 {class=MaterialExpressionNamedRerouteUsage}
  MaterialExpressionNamedRerouteUsage_8 {class=MaterialExpressionNamedRerouteUsage}
  MaterialExpressionNamedRerouteUsage_9 {class=MaterialExpressionNamedRerouteUsage}
  MaterialExpressionOneMinus_0 {class=MaterialExpressionOneMinus, inputs [None {from=MaterialExpressionNamedRerouteUsage_9}]}
  MaterialExpressionOneMinus_1 {class=MaterialExpressionOneMinus, inputs [None {from=MaterialExpressionSaturate_1}]}
  MaterialExpressionOneMinus_2 {class=MaterialExpressionOneMinus, inputs [None {from=MaterialExpressionSaturate_3}]}
  MaterialExpressionReroute_0 {class=MaterialExpressionReroute, inputs [None {from=MaterialExpressionScalarParameter_3}]}
  MaterialExpressionSaturate_0 {class=MaterialExpressionSaturate, inputs [None {from=MaterialExpressionSubtract_0}]}
  MaterialExpressionSaturate_1 {class=MaterialExpressionSaturate, inputs [None {from=MaterialExpressionSubtract_1}]}
  MaterialExpressionSaturate_2 {class=MaterialExpressionSaturate, inputs [None {from=MaterialExpressionAdd_4}]}
  MaterialExpressionSaturate_3 {class=MaterialExpressionSaturate, inputs [None {from=MaterialExpressionMultiply_19}]}
  MaterialExpressionScalarParameter_0 {class=MaterialExpressionScalarParameter, properties {defaultValue=9.999999843067494e+17, parameterName=BP_GetTime}}
  MaterialExpressionScalarParameter_1 {class=MaterialExpressionScalarParameter, properties {parameterName=BP_IsRunning}}
  MaterialExpressionScalarParameter_2 {class=MaterialExpressionScalarParameter, properties {defaultValue=1, parameterName=IsReactive}}
  MaterialExpressionScalarParameter_3 {class=MaterialExpressionScalarParameter, properties {parameterName=BP_IsFrontEnd}}
  MaterialExpressionSetMaterialAttributes_0 {
   class=MaterialExpressionSetMaterialAttributes
   inputs [
    MaterialAttributes {from=MaterialExpressionGetMaterialAttributes_0, output=MaterialAttributes}
    Base Color {from=MaterialExpressionLinearInterpolate_0, output=""}
    Emissive Color {from=MaterialExpressionAdd_1, output=""}
   ]
   properties {attributeSetTypes [#1, #2]}
  }
  MaterialExpressionSetMaterialAttributes_2 {
   class=MaterialExpressionSetMaterialAttributes
   inputs [
    MaterialAttributes {from=MaterialExpressionGetMaterialAttributes_0, output=MaterialAttributes}
    Base Color {from=MaterialExpressionMultiply_0, output=""}
    Emissive Color {from=MaterialExpressionMultiply_15, output=""}
   ]
   properties {attributeSetTypes [#1, #2]}
  }
  MaterialExpressionSine_0 {class=MaterialExpressionSine, inputs [None {from=MaterialExpressionMultiply_4}]}
  MaterialExpressionSmoothStep_0 {
   class=MaterialExpressionSmoothStep
   inputs [
    Min {from="", output=""}
    Max {from=MaterialExpressionConstant_7, output=""}
    Value {from=MaterialExpressionScalarParameter_1, output=""}
   ]
   properties {constMax=0.001}
  }
  MaterialExpressionSphereMask_0 {
   class=MaterialExpressionSphereMask
   inputs [
    A {from=MaterialExpressionComponentMask_1, output=""}
    B {from=MaterialExpressionConstant_2, output=""}
    Radius {from="", output=""}
    Hardness {from="", output=""}
   ]
   properties {attenuationRadius=1, hardnessPercent=0}
  }
  MaterialExpressionStep_1 {
   class=MaterialExpressionStep
   inputs [
    Y {from="", output=""}
    X {from=MaterialExpressionSaturate_0, output=""}
   ]
   properties {constY=0.05}
  }
  MaterialExpressionStep_2 {
   class=MaterialExpressionStep
   inputs [
    Y {from=MaterialExpressionConstant_7, output=""}
    X {from=MaterialExpressionScalarParameter_1, output=""}
   ]
   properties {constY=0.1}
  }
  MaterialExpressionSubtract_0 {
   class=MaterialExpressionSubtract
   inputs [
    A {from=MaterialExpressionTime_0}
    B {from=MaterialExpressionScalarParameter_0}
   ]
  }
  MaterialExpressionSubtract_1 {
   class=MaterialExpressionSubtract
   inputs [
    A {from=MaterialExpressionAdd_3}
    B {from=MaterialExpressionScalarParameter_0}
   ]
  }
  MaterialExpressionSubtract_2 {
   class=MaterialExpressionSubtract
   inputs [
    A {from=MaterialExpressionAdd_5}
    B {from=MaterialExpressionScalarParameter_0}
   ]
  }
  MaterialExpressionSwitch_0 {
   class=MaterialExpressionSwitch
   inputs [
    SwitchValue {from=MaterialExpressionNamedRerouteUsage_7, output=""}
    Default {from=MaterialExpressionConstant3Vector_2, output=""}
    None {from=MaterialExpressionConstant3Vector_2, output=""}
    None {from=MaterialExpressionConstant3Vector_1, output=""}
    None {from=MaterialExpressionConstant3Vector_0, output=""}
    None {from=MaterialExpressionConstant3Vector_3, output=""}
   ]
   properties {inputs [
     {
      inputName=None
      input {
       expression=&self:MaterialExpressionConstant3Vector_2
       inputName=None
       mask=1
       maskA=0
       maskB=1
       maskG=1
       maskR=1
       outputIndex=0
      }
     }
     {
      inputName=None
      input {
       expression=&self:MaterialExpressionConstant3Vector_1
       inputName=None
       mask=1
       maskA=0
       maskB=1
       maskG=1
       maskR=1
       outputIndex=0
      }
     }
     {
      inputName=None
      input {
       expression=&self:MaterialExpressionConstant3Vector_0
       inputName=None
       mask=1
       maskA=0
       maskB=1
       maskG=1
       maskR=1
       outputIndex=0
      }
     }
     {
      inputName=None
      input {
       expression=&self:MaterialExpressionConstant3Vector_3
       inputName=None
       mask=1
       maskA=0
       maskB=1
       maskG=1
       maskR=1
       outputIndex=0
      }
     }
    ]}
  }
  MaterialExpressionSwitch_1 {
   class=MaterialExpressionSwitch
   inputs [
    SwitchValue {from=MaterialExpressionNamedRerouteUsage_0, output=""}
    Default {from=MaterialExpressionConstant_3, output=""}
    None {from=MaterialExpressionConstant_3, output=""}
    None {from=MaterialExpressionConstant_4, output=""}
    None {from=MaterialExpressionConstant_5, output=""}
    None {from=MaterialExpressionConstant_6, output=""}
   ]
   properties {inputs [
     {
      inputName=None
      input {
       expression=&self:MaterialExpressionConstant_3
       inputName=None
       mask=0
       maskA=0
       maskB=0
       maskG=0
       maskR=0
       outputIndex=0
      }
     }
     {
      inputName=None
      input {
       expression=&self:MaterialExpressionConstant_4
       inputName=None
       mask=0
       maskA=0
       maskB=0
       maskG=0
       maskR=0
       outputIndex=0
      }
     }
     {
      inputName=None
      input {
       expression=&self:MaterialExpressionConstant_5
       inputName=None
       mask=0
       maskA=0
       maskB=0
       maskG=0
       maskR=0
       outputIndex=0
      }
     }
     {
      inputName=None
      input {
       expression=&self:MaterialExpressionConstant_6
       inputName=None
       mask=0
       maskA=0
       maskB=0
       maskG=0
       maskR=0
       outputIndex=0
      }
     }
    ]}
  }
  MaterialExpressionTime_0 {class=MaterialExpressionTime}
  MaterialExpressionTime_1 {class=MaterialExpressionTime}
  MaterialExpressionTime_2 {class=MaterialExpressionTime}
 }
 properties {
  bAutomaticallySetUsageInEditor=false
  bUseMaterialAttributes=true
  bUsedWithMeshDeformer=true
  bUsedWithMorphTargets=true
  bUsedWithSkeletalMesh=true
  pixelDepthOffsetMode=PDOM_Legacy
  shadingModel=MSM_FromMaterialExpression
  usedShadingModels=Default Lit | Subsurface | Clear Coat | Subsurface Profile | Cloth
 }
}
