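# Physics asset for the medium male GrindRex body that sets up secondary physics for a dangling belt strap and keychain.
# Front, back and master strap chains hang from the pelvis on zero-size base capsules with tight angular limits, cross-linked at their tips into closed loops.
# The two-link keychain has a free-swinging base and limited links.
# Capsules on pelvis, lower spine, both thighs and left calf and foot provide body shapes.
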PhysicsAsset /BRCosmetics/Characters/Player/Male/Medium/Bodies/M_MED_GrindRex/Meshes/M_MED_GrindRex_Physics_BeltStrapKeyChain
bodies {
 belt_strap_bk_02 [{shapeType=Capsule, length=5.1321, radius=2.3112, shapeName=belt_strap_bk_02_capsule, center=<4.2616 -1.3512 0>, rotation=<rot 90 0 0>}]
 belt_strap_master_01 [{shapeType=Capsule, length=10, radius=4, shapeName=belt_strap_master_01_capsule, center=<7.1391 0 0>, rotation=<rot 90 0 0>}]
 belt_strap_master_02 [{shapeType=Capsule, length=10, radius=4, shapeName=belt_strap_master_02_capsule, center=<7.1391 0 0>, rotation=<rot 90 0 0>}]
 belt_strap_master_03 [{shapeType=Capsule, length=9.8632, radius=4.847, shapeName=belt_strap_master_03_capsule, center=<1.4849 0 3.193>, rotation=<rot 0 0 90>}]
 belt_strap_master_base [{shapeType=Capsule, length=0, radius=0, shapeName=belt_strap_master_base_capsule}]
 calf_l [{shapeType=Capsule, length=34.0481, radius=5.3358, shapeName=calf_l_capsule, center=<-16.5379 0 0>, rotation=<rot 90 0 0>}]
 dyn_belt_strap_bk_01 [{shapeType=Capsule, length=5.1321, radius=2.3112, shapeName=dyn_belt_strap_bk_01_capsule, center=<4.2616 -1.4539 0>, rotation=<rot 90 0 0>}]
 dyn_belt_strap_bk_03 [{shapeType=Capsule, length=5.1321, radius=2.3112, shapeName=dyn_belt_strap_bk_03_capsule, center=<4.2616 -1.3512 0>, rotation=<rot 90 0 0>}]
 dyn_belt_strap_bk_04 [{shapeType=Capsule, length=5.1321, radius=2.3112, shapeName=dyn_belt_strap_bk_04_capsule, center=<4.2616 -1.3512 0>, rotation=<rot 90 0 0>}]
 dyn_belt_strap_bk_05 [{shapeType=Capsule, length=5.1321, radius=2.3112, shapeName=dyn_belt_strap_bk_05_capsule, center=<4.2616 -1.5728 0>, rotation=<rot 90 0 0>}]
 dyn_belt_strap_bk_06 [{shapeType=Capsule, length=5.1321, radius=2.3112, shapeName=dyn_belt_strap_bk_06_capsule, center=<4.2616 -1.3512 0>, rotation=<rot 90 0 0>}]
 dyn_belt_strap_bk_base [{shapeType=Capsule, length=0, radius=0, shapeName=dyn_belt_strap_bk_base_capsule}]
 dyn_belt_strap_frnt_01 [{shapeType=Capsule, length=2.4652, radius=2.2095, shapeName=dyn_belt_strap_frnt_01_capsule, center=<2.8656 0.6728 0>, rotation=<rot 90 0 0>}]
 dyn_belt_strap_frnt_02 [{shapeType=Capsule, length=2.4652, radius=2.2095, shapeName=dyn_belt_strap_frnt_02_capsule, center=<2.8656 0.6728 0>, rotation=<rot 90 0 0>}]
 dyn_belt_strap_frnt_03 [{shapeType=Capsule, length=2.4652, radius=2.2095, shapeName=dyn_belt_strap_frnt_03_capsule, center=<2.8656 0.6728 0>, rotation=<rot 90 0 0>}]
 dyn_belt_strap_frnt_04 [{shapeType=Capsule, length=2.4652, radius=2.2095, shapeName=dyn_belt_strap_frnt_04_capsule, center=<2.8656 0.6728 0>, rotation=<rot 90 0 0>}]
 dyn_belt_strap_frnt_05 [{shapeType=Capsule, length=6.5815, radius=2.8516, shapeName=dyn_belt_strap_frnt_05_capsule, center=<2.8656 0.9642 0>, rotation=<rot 90 0 0>}]
 dyn_belt_strap_frnt_base [{shapeType=Capsule, length=0, radius=0, shapeName=dyn_belt_strap_frnt_base_capsule}]
 dyn_keychain_01 [{shapeType=Capsule, length=2.4344, radius=1.564, shapeName=dyn_keychain_01_capsule, center=<1.953 0.8547 0>, rotation=<rot 90 0 0>}]
 dyn_keychain_02 [{shapeType=Capsule, length=1.125, radius=4.8401, shapeName=dyn_keychain_02_capsule, center=<6.603 1.8374 0>, rotation=<rot 90 0 0>}]
 dyn_keychain_base [{shapeType=Capsule, length=0, radius=0, shapeName=dyn_keychain_base_capsule}]
 foot_l [{shapeType=Capsule, length=12.9606, radius=10.4142, shapeName=foot_l_capsule, center=<-3.8758 6.0723 0>, rotation=<rot 0 7.8954 90>}]
 pelvis [{shapeType=Capsule, length=10.6398, radius=11.8495, shapeName=pelvis_capsule, center=<-7.3603 0.0873 0>}]
 spine_01 [{shapeType=Capsule, length=11.9127, radius=10.3205, shapeName=spine_01_capsule, center=<-4.3492 0 0>}]
 thigh_l [{shapeType=Capsule, length=25.6627, radius=9.1411, shapeName=thigh_l_capsule, center=<-16.5017 -0.929 0>, rotation=<rot 90 -3.5349 0>}]
 thigh_r [{shapeType=Capsule, length=27.4265, radius=7.3773, shapeName=thigh_r_capsule, center=<17.5395 0 0>, rotation=<rot 90 180 0>}]
}
constraints [
 {bone1Name=belt_strap_master_base, bone2Name=pelvis, swing1LimitDegrees=10, swing2LimitDegrees=10, twistLimitDegrees=5}
 {bone1Name=belt_strap_master_01, bone2Name=belt_strap_master_base, swing1LimitDegrees=10, swing2LimitDegrees=10, twistLimitDegrees=5}
 {bone1Name=belt_strap_master_02, bone2Name=belt_strap_master_01, swing1LimitDegrees=10, swing2LimitDegrees=10, twistLimitDegrees=5}
 {bone1Name=belt_strap_master_03, bone2Name=belt_strap_master_02, swing1LimitDegrees=10, swing2LimitDegrees=10, twistLimitDegrees=5}
 {bone1Name=dyn_belt_strap_bk_05, bone2Name=dyn_belt_strap_bk_04, swing1LimitDegrees=10, swing2LimitDegrees=10, twistLimitDegrees=5}
 {bone1Name=dyn_belt_strap_bk_04, bone2Name=dyn_belt_strap_bk_03, swing1LimitDegrees=10, swing2LimitDegrees=10, twistLimitDegrees=5}
 {bone1Name=dyn_belt_strap_bk_03, bone2Name=belt_strap_bk_02, swing1LimitDegrees=10, swing2LimitDegrees=10, twistLimitDegrees=5}
 {bone1Name=belt_strap_bk_02, bone2Name=dyn_belt_strap_bk_01, swing1LimitDegrees=10, swing2LimitDegrees=10, twistLimitDegrees=5}
 {bone1Name=dyn_belt_strap_bk_base, bone2Name=pelvis, swing1LimitDegrees=10, swing2LimitDegrees=10, twistLimitDegrees=5}
 {bone1Name=dyn_belt_strap_bk_01, bone2Name=dyn_belt_strap_bk_base, swing1LimitDegrees=10, swing2LimitDegrees=10, twistLimitDegrees=5}
 {bone1Name=dyn_belt_strap_frnt_base, bone2Name=pelvis, swing1LimitDegrees=10, swing2LimitDegrees=10, twistLimitDegrees=5}
 {bone1Name=dyn_belt_strap_frnt_01, bone2Name=dyn_belt_strap_frnt_base, swing1LimitDegrees=10, swing2LimitDegrees=10, twistLimitDegrees=5}
 {bone1Name=dyn_belt_strap_frnt_02, bone2Name=dyn_belt_strap_frnt_01, swing1LimitDegrees=10, swing2LimitDegrees=10, twistLimitDegrees=5}
 {bone1Name=dyn_belt_strap_frnt_03, bone2Name=dyn_belt_strap_frnt_02, swing1LimitDegrees=10, swing2LimitDegrees=10, twistLimitDegrees=5}
 {bone1Name=dyn_belt_strap_frnt_04, bone2Name=dyn_belt_strap_frnt_03, swing1LimitDegrees=10, swing2LimitDegrees=10, twistLimitDegrees=5}
 {bone1Name=dyn_belt_strap_frnt_05, bone2Name=dyn_belt_strap_frnt_04, swing1LimitDegrees=10, swing2LimitDegrees=10, twistLimitDegrees=5}
 {
  bone1Name=dyn_keychain_base
  bone2Name=pelvis
  swing1LimitDegrees=45
  swing1Motion=Free
  swing2LimitDegrees=45
  swing2Motion=Free
  twistLimitDegrees=45
  twistMotion=Free
 }
 {bone1Name=dyn_keychain_01, bone2Name=dyn_keychain_base, swing1LimitDegrees=30, swing2LimitDegrees=2, twistLimitDegrees=10}
 {bone1Name=dyn_keychain_02, bone2Name=dyn_keychain_01, swing1LimitDegrees=20, swing2LimitDegrees=20, twistLimitDegrees=10}
 {
  bone1Name=spine_01
  bone2Name=pelvis
  swing1LimitDegrees=45
  swing1Motion=Free
  swing2LimitDegrees=45
  swing2Motion=Free
  twistLimitDegrees=45
  twistMotion=Free
 }
 {bone1Name=dyn_belt_strap_bk_06, bone2Name=dyn_belt_strap_bk_05, swing1LimitDegrees=10, swing2LimitDegrees=10, twistLimitDegrees=5}
 {
  bone1Name=thigh_l
  bone2Name=pelvis
  swing1LimitDegrees=45
  swing1Motion=Free
  swing2LimitDegrees=45
  swing2Motion=Free
  twistLimitDegrees=45
  twistMotion=Free
 }
 {
  bone1Name=thigh_r
  bone2Name=pelvis
  swing1LimitDegrees=45
  swing1Motion=Free
  swing2LimitDegrees=45
  swing2Motion=Free
  twistLimitDegrees=45
  twistMotion=Free
 }
 {
  bone1Name=calf_l
  bone2Name=thigh_l
  swing1LimitDegrees=45
  swing1Motion=Free
  swing2LimitDegrees=45
  swing2Motion=Free
  twistLimitDegrees=45
  twistMotion=Free
 }
 {
  bone1Name=foot_l
  bone2Name=calf_l
  swing1LimitDegrees=45
  swing1Motion=Free
  swing2LimitDegrees=45
  swing2Motion=Free
  twistLimitDegrees=45
  twistMotion=Free
 }
 {bone1Name=dyn_belt_strap_frnt_05, bone2Name=belt_strap_master_03, swing1LimitDegrees=10, swing2LimitDegrees=10, twistLimitDegrees=5}
 {bone1Name=dyn_belt_strap_bk_06, bone2Name=belt_strap_master_03, swing1LimitDegrees=10, swing2LimitDegrees=10, twistLimitDegrees=5}
 {bone1Name=dyn_belt_strap_frnt_05, bone2Name=dyn_belt_strap_bk_06, swing1LimitDegrees=10, swing2LimitDegrees=10, twistLimitDegrees=5}
]
