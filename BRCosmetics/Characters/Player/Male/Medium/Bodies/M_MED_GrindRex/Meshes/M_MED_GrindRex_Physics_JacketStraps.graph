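# Physics asset for the two front jacket straps on the GrindRex medium male player body.
# Each strap is a five-body capsule chain, a small base and four segments, constrained from spine_05 with narrow swing and twist limits that differ slightly per side.
# Capsules on the upper spine, clavicles and upper arms form a body proxy around the straps, linked by free-rotation constraints.
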PhysicsAsset /BRCosmetics/Characters/Player/Male/Medium/Bodies/M_MED_GrindRex/Meshes/M_MED_GrindRex_Physics_JacketStraps
bodies {
 clavicle_l [{shapeType=Capsule, length=12.1632, radius=6.9307, shapeName=clavicle_l_capsule, center=<7.6529 0 0>, rotation=<rot 90 0 0>}]
 clavicle_r [{shapeType=Capsule, length=12.1632, radius=6.9307, shapeName=clavicle_r_capsule, center=<-7.6529 0 0>, rotation=<rot 90 180 0>}]
 dyn_jacket_strap_fr_l_01 [{shapeType=Capsule, length=4, radius=2, shapeName=dyn_jacket_strap_fr_l_01_capsule, center=<3.3023 0 0>, rotation=<rot 90 0 0>}]
 dyn_jacket_strap_fr_l_02 [{shapeType=Capsule, length=4, radius=2, shapeName=dyn_jacket_strap_fr_l_02_capsule, center=<3.3023 0 0>, rotation=<rot 90 0 0>}]
 dyn_jacket_strap_fr_l_03 [{shapeType=Capsule, length=4, radius=2, shapeName=dyn_jacket_strap_fr_l_03_capsule, center=<3.3023 0 0>, rotation=<rot 90 0 0>}]
 dyn_jacket_strap_fr_l_04 [{shapeType=Capsule, length=4, radius=2, shapeName=dyn_jacket_strap_fr_l_04_capsule, center=<3.3023 0 0>, rotation=<rot 90 0 0>}]
 dyn_jacket_strap_fr_l_base [{shapeType=Capsule, length=0.505, radius=0.505, shapeName=dyn_jacket_strap_fr_l_base_capsule, rotation=<rot 0 -90 0>}]
 dyn_jacket_strap_fr_r_01 [{shapeType=Capsule, length=4, radius=2, shapeName=dyn_jacket_strap_fr_r_01_capsule, center=<-3.0425 0 0>, rotation=<rot 90 0 0>}]
 dyn_jacket_strap_fr_r_02 [{shapeType=Capsule, length=4, radius=2, shapeName=dyn_jacket_strap_fr_r_02_capsule, center=<-3.0425 0 0>, rotation=<rot 90 0 0>}]
 dyn_jacket_strap_fr_r_03 [{shapeType=Capsule, length=4, radius=2, shapeName=dyn_jacket_strap_fr_r_03_capsule, center=<-3.0425 0 0>, rotation=<rot 90 0 0>}]
 dyn_jacket_strap_fr_r_04 [{shapeType=Capsule, length=4, radius=2, shapeName=dyn_jacket_strap_fr_r_04_capsule, center=<-3.0425 0 0>, rotation=<rot 90 0 0>}]
 dyn_jacket_strap_fr_r_base [{shapeType=Capsule, length=0.505, radius=0.505, shapeName=dyn_jacket_strap_fr_r_base_capsule, rotation=<rot 0 -90 0>}]
 spine_04 [{shapeType=Capsule, length=11.835, radius=13.942, shapeName=spine_04_capsule, center=<-1.8374 -4.2822 0>}]
 spine_05 [{shapeType=Capsule, length=9.7916, radius=14.6616, shapeName=spine_05_capsule, center=<-5.8217 -6.7515 0>}]
 upperarm_l [{shapeType=Capsule, length=23.5072, radius=6.8796, shapeName=upperarm_l_capsule, center=<12.9678 0 0>, rotation=<rot 90 0 0>}]
 upperarm_r [{shapeType=Capsule, length=23.0694, radius=7.3174, shapeName=upperarm_r_capsule, center=<-12.9678 0 0>, rotation=<rot 90 180 0>}]
}
constraints [
 {
  bone1Name=clavicle_l
  bone2Name=spine_05
  swing1LimitDegrees=45
  swing1Motion=Free
  swing2LimitDegrees=45
  swing2Motion=Free
  twistLimitDegrees=45
  twistMotion=Free
 }
 {
  bone1Name=upperarm_l
  bone2Name=clavicle_l
  swing1LimitDegrees=45
  swing1Motion=Free
  swing2LimitDegrees=45
  swing2Motion=Free
  twistLimitDegrees=45
  twistMotion=Free
 }
 {
  bone1Name=clavicle_r
  bone2Name=spine_05
  swing1LimitDegrees=45
  swing1Motion=Free
  swing2LimitDegrees=45
  swing2Motion=Free
  twistLimitDegrees=45
  twistMotion=Free
 }
 {
  bone1Name=upperarm_r
  bone2Name=clavicle_r
  swing1LimitDegrees=45
  swing1Motion=Free
  swing2LimitDegrees=45
  swing2Motion=Free
  twistLimitDegrees=45
  twistMotion=Free
 }
 {
  bone1Name=spine_05
  bone2Name=spine_04
  swing1LimitDegrees=45
  swing1Motion=Free
  swing2LimitDegrees=45
  swing2Motion=Free
  twistLimitDegrees=45
  twistMotion=Free
 }
 {bone1Name=dyn_jacket_strap_fr_l_base, bone2Name=spine_05, swing1LimitDegrees=20, swing2LimitDegrees=20, twistLimitDegrees=5}
 {bone1Name=dyn_jacket_strap_fr_l_01, bone2Name=dyn_jacket_strap_fr_l_base, swing1LimitDegrees=10, swing2LimitDegrees=20, twistLimitDegrees=5}
 {bone1Name=dyn_jacket_strap_fr_l_02, bone2Name=dyn_jacket_strap_fr_l_01, swing1LimitDegrees=10, swing2LimitDegrees=20, twistLimitDegrees=5}
 {bone1Name=dyn_jacket_strap_fr_l_03, bone2Name=dyn_jacket_strap_fr_l_02, swing1LimitDegrees=10, swing2LimitDegrees=20, twistLimitDegrees=5}
 {bone1Name=dyn_jacket_strap_fr_l_04, bone2Name=dyn_jacket_strap_fr_l_03, swing1LimitDegrees=10, swing2LimitDegrees=40, twistLimitDegrees=5}
 {bone1Name=dyn_jacket_strap_fr_r_base, bone2Name=spine_05, swing1LimitDegrees=20, swing2LimitDegrees=20, twistLimitDegrees=5}
 {bone1Name=dyn_jacket_strap_fr_r_01, bone2Name=dyn_jacket_strap_fr_r_base, swing1LimitDegrees=10, swing2LimitDegrees=20, twistLimitDegrees=5}
 {bone1Name=dyn_jacket_strap_fr_r_02, bone2Name=dyn_jacket_strap_fr_r_01, swing1LimitDegrees=15, swing2LimitDegrees=20, twistLimitDegrees=5}
 {bone1Name=dyn_jacket_strap_fr_r_03, bone2Name=dyn_jacket_strap_fr_r_02, swing1LimitDegrees=15, swing2LimitDegrees=30, twistLimitDegrees=5}
 {bone1Name=dyn_jacket_strap_fr_r_04, bone2Name=dyn_jacket_strap_fr_r_03, swing1LimitDegrees=15, swing2LimitDegrees=20, twistLimitDegrees=5}
]
