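# Physics asset for the GrindRex medium male body's jacket, belt strap and keychain, named as the Frontend variant.
# Capsules cover the dynamic bone chains, plus pelvis, spine and thighs as body shapes.
# Jacket chains hang from spine_02 with tight per-segment angular limits, while strap and keychain hang from pelvis with free rotation.
# Free-rotation cross-constraints tie front hem ends to the sides, and right side to back only, with one front-right link duplicated.
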PhysicsAsset /BRCosmetics/Characters/Player/Male/Medium/Bodies/M_MED_GrindRex/Meshes/M_MED_GrindRex_Physics_Jacket_Frontend
bodies {
 dyn_belt_strap_frnt_01 [{shapeType=Capsule, length=5.5742, radius=1.4258, shapeName=dyn_belt_strap_frnt_01_capsule, center=<4.4539 0 0>, rotation=<rot 90 0 0>}]
 dyn_belt_strap_frnt_02 [{shapeType=Capsule, length=5.5742, radius=1.4258, shapeName=dyn_belt_strap_frnt_02_capsule, center=<4.4539 0 0>, rotation=<rot 90 0 0>}]
 dyn_belt_strap_frnt_03 [{shapeType=Capsule, length=5.2093, radius=1.7907, shapeName=dyn_belt_strap_frnt_03_capsule, center=<4.4539 0 0>, rotation=<rot 90 0 0>}]
 dyn_belt_strap_frnt_04 [{shapeType=Capsule, length=5.2093, radius=1.7907, shapeName=dyn_belt_strap_frnt_04_capsule, center=<4.4539 0 0>, rotation=<rot 90 0 0>}]
 dyn_belt_strap_frnt_05 [{shapeType=Capsule, length=5.2093, radius=1.7907, shapeName=dyn_belt_strap_frnt_05_capsule, center=<4.4539 0 0>, rotation=<rot 90 0 0>}]
 dyn_belt_strap_frnt_base [{shapeType=Capsule, length=0, radius=0, shapeName=dyn_belt_strap_frnt_base_capsule, center=<4.4539 0 0>, rotation=<rot 90 0 0>}]
 dyn_jacket_bk_l_01 [{shapeType=Capsule, length=13.3929, radius=4, shapeName=dyn_jacket_bk_l_01_capsule, center=<3.8156 0 0>}]
 dyn_jacket_bk_l_02 [{shapeType=Capsule, length=13.9467, radius=4.912, shapeName=dyn_jacket_bk_l_02_capsule, center=<3.8156 0 0>}]
 dyn_jacket_bk_l_base [{shapeType=Capsule, length=0.505, radius=0.505, shapeName=dyn_jacket_bk_l_base_capsule}]
 dyn_jacket_bk_r_01 [{shapeType=Capsule, length=13.3929, radius=4, shapeName=dyn_jacket_bk_r_01_capsule, center=<-3.8156 0 0>, rotation=<rot 0 0 -180>}]
 dyn_jacket_bk_r_02 [{shapeType=Capsule, length=13.9467, radius=4.912, shapeName=dyn_jacket_bk_r_02_capsule, center=<-3.8156 0 0>, rotation=<rot 0 0 -180>}]
 dyn_jacket_bk_r_base [{shapeType=Capsule, length=0.505, radius=0.505, shapeName=dyn_jacket_bk_r_base_capsule}]
 dyn_jacket_fr_l_01 [{shapeType=Capsule, length=14.884, radius=3.9886, shapeName=dyn_jacket_fr_l_01_capsule, center=<3.8277 0 0>}]
 dyn_jacket_fr_l_02 [{shapeType=Capsule, length=15.2884, radius=3.9544, shapeName=dyn_jacket_fr_l_02_capsule, center=<3.8277 0 0>}]
 dyn_jacket_fr_l_03 [{
   shapeType=Capsule
   length=15.7261
   radius=3.6601
   shapeName=dyn_jacket_fr_l_03_capsule
   center=<2.7323 1.4009 0>
   rotation=<rot -0.0001 0.0001 4.6061>
  }]
 dyn_jacket_fr_l_base [{shapeType=Capsule, length=0.505, radius=0.505, shapeName=dyn_jacket_fr_l_base_capsule}]
 dyn_jacket_fr_r_01 [{shapeType=Capsule, length=14.6298, radius=4.2428, shapeName=dyn_jacket_fr_r_01_capsule, center=<-3.8277 0 0>, rotation=<rot 0 0 -180>}]
 dyn_jacket_fr_r_02 [{shapeType=Capsule, length=14.9152, radius=3.9574, shapeName=dyn_jacket_fr_r_02_capsule, center=<-3.8277 0 0>, rotation=<rot 0 0 -180>}]
 dyn_jacket_fr_r_03 [{shapeType=Capsule, length=16.1143, radius=4.1838, shapeName=dyn_jacket_fr_r_03_capsule, center=<-3.8277 -2.6139 0>, rotation=<rot 0 0 -180>}]
 dyn_jacket_fr_r_base [{shapeType=Capsule, length=0.505, radius=0.505, shapeName=dyn_jacket_fr_r_base_capsule}]
 dyn_jacket_mid_01 [{shapeType=Capsule, length=15.1448, radius=4, shapeName=dyn_jacket_mid_01_capsule, center=<4.1515 0 0>}]
 dyn_jacket_mid_02 [{shapeType=Capsule, length=13.9938, radius=5.151, shapeName=dyn_jacket_mid_02_capsule, center=<5.1557 0 0>}]
 dyn_jacket_mid_base [{shapeType=Capsule, length=0.505, radius=0.505, shapeName=dyn_jacket_mid_base_capsule}]
 dyn_jacket_side_l_01 [{shapeType=Capsule, length=17.4012, radius=4.5163, shapeName=dyn_jacket_side_l_01_capsule, center=<3.3341 0 0>}]
 dyn_jacket_side_l_02 [{shapeType=Capsule, length=18.1824, radius=3.7351, shapeName=dyn_jacket_side_l_02_capsule, center=<3.3341 0 0>}]
 dyn_jacket_side_l_03 [{shapeType=Capsule, length=19.6905, radius=2.2271, shapeName=dyn_jacket_side_l_03_capsule, center=<3.3341 0 0>}]
 dyn_jacket_side_l_base [{shapeType=Capsule, length=0.505, radius=0.505, shapeName=dyn_jacket_side_l_base_capsule}]
 dyn_jacket_side_r_01 [{shapeType=Capsule, length=17.4012, radius=4.5163, shapeName=dyn_jacket_side_r_01_capsule, center=<-3.3341 0 0>, rotation=<rot 0 0 -180>}]
 dyn_jacket_side_r_02 [{shapeType=Capsule, length=17.7042, radius=4.2133, shapeName=dyn_jacket_side_r_02_capsule, center=<-3.3341 0 0>, rotation=<rot 0 0 -180>}]
 dyn_jacket_side_r_03 [{shapeType=Capsule, length=20.5251, radius=1.3924, shapeName=dyn_jacket_side_r_03_capsule, center=<-3.3341 0 0>, rotation=<rot 0 0 -180>}]
 dyn_jacket_side_r_base [{shapeType=Capsule, length=0.505, radius=0.505, shapeName=dyn_jacket_side_r_base_capsule}]
 dyn_keychain_01 [{shapeType=Capsule, length=4.8407, radius=2.1194, shapeName=dyn_keychain_01_capsule, center=<3.1489 0 0>, rotation=<rot 90 0 0>}]
 dyn_keychain_02 [{shapeType=Capsule, length=3.7971, radius=4.8231, shapeName=dyn_keychain_02_capsule, center=<7.5451 1.5411 0>, rotation=<rot 90 19.0978 0>}]
 dyn_keychain_base [{shapeType=Capsule, length=0.505, radius=0.505, shapeName=dyn_keychain_base_capsule, rotation=<rot 0 -90 0>}]
 pelvis [{shapeType=Capsule, length=15.3851, radius=9.3668, shapeName=pelvis_capsule, center=<-7.3479 -2.2662 0>}]
 spine_01 [{shapeType=Capsule, length=15.3176, radius=8.9624, shapeName=spine_01_capsule, center=<-7.746 0 0>}]
 spine_02 [{shapeType=Capsule, length=11.5353, radius=10.0885, shapeName=spine_02_capsule, center=<0 -1.3544 0>}]
 thigh_l [{shapeType=Capsule, length=50.5018, radius=8.2304, shapeName=thigh_l_capsule, center=<-23.4261 0.0927 0.1783>, rotation=<rot 90 -1.999 0>}]
 thigh_r [{shapeType=Capsule, length=51.2526, radius=7.4796, shapeName=thigh_r_capsule, center=<23.4261 0.7912 0>, rotation=<rot 90 180 0>}]
}
constraints [
 {
  bone1Name=spine_01
  bone2Name=pelvis
  swing1LimitDegrees=45
  swing1Motion=Free
  swing2LimitDegrees=45
  swing2Motion=Free
  twistLimitDegrees=45
  twistMotion=Free
 }
 {
  bone1Name=spine_02
  bone2Name=spine_01
  swing1LimitDegrees=45
  swing1Motion=Free
  swing2LimitDegrees=45
  swing2Motion=Free
  twistLimitDegrees=45
  twistMotion=Free
 }
 {
  bone1Name=dyn_jacket_bk_l_base
  bone2Name=spine_02
  swing1LimitDegrees=45
  swing1Motion=Free
  swing2LimitDegrees=45
  swing2Motion=Free
  twistLimitDegrees=45
  twistMotion=Free
 }
 {bone1Name=dyn_jacket_bk_l_01, bone2Name=dyn_jacket_bk_l_base, swing1LimitDegrees=10, swing2LimitDegrees=0, twistLimitDegrees=0}
 {bone1Name=dyn_jacket_bk_l_02, bone2Name=dyn_jacket_bk_l_01, swing1LimitDegrees=10, swing2LimitDegrees=0, twistLimitDegrees=0}
 {
  bone1Name=dyn_jacket_bk_r_base
  bone2Name=spine_02
  swing1LimitDegrees=45
  swing1Motion=Free
  swing2LimitDegrees=45
  swing2Motion=Free
  twistLimitDegrees=45
  twistMotion=Free
 }
 {bone1Name=dyn_jacket_bk_r_01, bone2Name=dyn_jacket_bk_r_base, swing1LimitDegrees=10, swing2LimitDegrees=0, twistLimitDegrees=0}
 {bone1Name=dyn_jacket_bk_r_02, bone2Name=dyn_jacket_bk_r_01, swing1LimitDegrees=10, swing2LimitDegrees=5, twistLimitDegrees=0}
 {
  bone1Name=dyn_jacket_fr_l_base
  bone2Name=spine_02
  swing1LimitDegrees=45
  swing1Motion=Free
  swing2LimitDegrees=45
  swing2Motion=Free
  twistLimitDegrees=45
  twistMotion=Free
 }
 {bone1Name=dyn_jacket_fr_l_01, bone2Name=dyn_jacket_fr_l_base, swing1LimitDegrees=45, swing2LimitDegrees=0, twistLimitDegrees=10}
 {bone1Name=dyn_jacket_fr_l_02, bone2Name=dyn_jacket_fr_l_01, swing1LimitDegrees=45, swing2LimitDegrees=0, twistLimitDegrees=20}
 {bone1Name=dyn_jacket_fr_l_03, bone2Name=dyn_jacket_fr_l_02, swing1LimitDegrees=10, swing2LimitDegrees=5, twistLimitDegrees=0}
 {
  bone1Name=dyn_jacket_fr_r_base
  bone2Name=spine_02
  swing1LimitDegrees=45
  swing1Motion=Free
  swing2LimitDegrees=45
  swing2Motion=Free
  twistLimitDegrees=45
  twistMotion=Free
 }
 {bone1Name=dyn_jacket_fr_r_01, bone2Name=dyn_jacket_fr_r_base, swing1LimitDegrees=45, swing2LimitDegrees=5, twistLimitDegrees=0}
 {bone1Name=dyn_jacket_fr_r_02, bone2Name=dyn_jacket_fr_r_01, swing1LimitDegrees=45, swing2LimitDegrees=0, twistLimitDegrees=0}
 {bone1Name=dyn_jacket_fr_r_03, bone2Name=dyn_jacket_fr_r_02, swing1LimitDegrees=20, swing2LimitDegrees=5, twistLimitDegrees=0}
 {
  bone1Name=dyn_jacket_mid_base
  bone2Name=spine_02
  swing1LimitDegrees=45
  swing1Motion=Free
  swing2LimitDegrees=45
  swing2Motion=Free
  twistLimitDegrees=45
  twistMotion=Free
 }
 {bone1Name=dyn_jacket_mid_01, bone2Name=dyn_jacket_mid_base, swing1LimitDegrees=10, swing2LimitDegrees=2, twistLimitDegrees=0}
 {bone1Name=dyn_jacket_mid_02, bone2Name=dyn_jacket_mid_01, swing1LimitDegrees=10, swing2LimitDegrees=2, twistLimitDegrees=0}
 {
  bone1Name=dyn_jacket_side_l_base
  bone2Name=spine_02
  swing1LimitDegrees=45
  swing1Motion=Free
  swing2LimitDegrees=45
  swing2Motion=Free
  twistLimitDegrees=45
  twistMotion=Free
 }
 {bone1Name=dyn_jacket_side_l_01, bone2Name=dyn_jacket_side_l_base, swing1LimitDegrees=5, swing2LimitDegrees=10, twistLimitDegrees=0}
 {bone1Name=dyn_jacket_side_l_02, bone2Name=dyn_jacket_side_l_01, swing1LimitDegrees=5, swing2LimitDegrees=20, twistLimitDegrees=10}
 {bone1Name=dyn_jacket_side_l_03, bone2Name=dyn_jacket_side_l_02, swing1LimitDegrees=5, swing2LimitDegrees=30, twistLimitDegrees=30}
 {
  bone1Name=dyn_jacket_side_r_base
  bone2Name=spine_02
  swing1LimitDegrees=45
  swing1Motion=Free
  swing2LimitDegrees=45
  swing2Motion=Free
  twistLimitDegrees=45
  twistMotion=Free
 }
 {bone1Name=dyn_jacket_side_r_01, bone2Name=dyn_jacket_side_r_base, swing1LimitDegrees=5, swing2LimitDegrees=25, twistLimitDegrees=0}
 {bone1Name=dyn_jacket_side_r_02, bone2Name=dyn_jacket_side_r_01, swing1LimitDegrees=10, swing2LimitDegrees=10, twistLimitDegrees=10}
 {bone1Name=dyn_jacket_side_r_03, bone2Name=dyn_jacket_side_r_02, swing1LimitDegrees=10, swing2LimitDegrees=10, twistLimitDegrees=30}
 {
  bone1Name=thigh_l
  bone2Name=pelvis
  swing1LimitDegrees=45
  swing1Motion=Free
  swing2LimitDegrees=45
  swing2Motion=Free
  twistLimitDegrees=45
  twistMotion=Free
 }
 {
  bone1Name=thigh_r
  bone2Name=pelvis
  swing1LimitDegrees=45
  swing1Motion=Free
  swing2LimitDegrees=45
  swing2Motion=Free
  twistLimitDegrees=45
  twistMotion=Free
 }
 {
  bone1Name=dyn_jacket_fr_l_03
  bone2Name=dyn_jacket_side_l_03
  swing1LimitDegrees=45
  swing1Motion=Free
  swing2LimitDegrees=45
  swing2Motion=Free
  twistLimitDegrees=45
  twistMotion=Free
 }
 {
  bone1Name=dyn_jacket_fr_r_03
  bone2Name=dyn_jacket_side_r_03
  swing1LimitDegrees=45
  swing1Motion=Free
  swing2LimitDegrees=45
  swing2Motion=Free
  twistLimitDegrees=45
  twistMotion=Free
 }
 {
  bone1Name=dyn_keychain_base
  bone2Name=pelvis
  swing1LimitDegrees=45
  swing1Motion=Free
  swing2LimitDegrees=45
  swing2Motion=Free
  twistLimitDegrees=45
  twistMotion=Free
 }
 {
  bone1Name=dyn_keychain_01
  bone2Name=dyn_keychain_base
  swing1LimitDegrees=45
  swing1Motion=Free
  swing2LimitDegrees=45
  swing2Motion=Free
  twistLimitDegrees=45
  twistMotion=Free
 }
 {
  bone1Name=dyn_keychain_02
  bone2Name=dyn_keychain_01
  swing1LimitDegrees=45
  swing1Motion=Free
  swing2LimitDegrees=45
  swing2Motion=Free
  twistLimitDegrees=45
  twistMotion=Free
 }
 {
  bone1Name=dyn_belt_strap_frnt_base
  bone2Name=pelvis
  swing1LimitDegrees=45
  swing1Motion=Free
  swing2LimitDegrees=45
  swing2Motion=Free
  twistLimitDegrees=45
  twistMotion=Free
 }
 {
  bone1Name=dyn_belt_strap_frnt_01
  bone2Name=dyn_belt_strap_frnt_base
  swing1LimitDegrees=45
  swing1Motion=Free
  swing2LimitDegrees=45
  swing2Motion=Free
  twistLimitDegrees=45
  twistMotion=Free
 }
 {
  bone1Name=dyn_belt_strap_frnt_02
  bone2Name=dyn_belt_strap_frnt_01
  swing1LimitDegrees=45
  swing1Motion=Free
  swing2LimitDegrees=45
  swing2Motion=Free
  twistLimitDegrees=45
  twistMotion=Free
 }
 {
  bone1Name=dyn_belt_strap_frnt_03
  bone2Name=dyn_belt_strap_frnt_02
  swing1LimitDegrees=45
  swing1Motion=Free
  swing2LimitDegrees=45
  swing2Motion=Free
  twistLimitDegrees=45
  twistMotion=Free
 }
 {
  bone1Name=dyn_belt_strap_frnt_04
  bone2Name=dyn_belt_strap_frnt_03
  swing1LimitDegrees=45
  swing1Motion=Free
  swing2LimitDegrees=45
  swing2Motion=Free
  twistLimitDegrees=45
  twistMotion=Free
 }
 {
  bone1Name=dyn_belt_strap_frnt_05
  bone2Name=dyn_belt_strap_frnt_04
  swing1LimitDegrees=45
  swing1Motion=Free
  swing2LimitDegrees=45
  swing2Motion=Free
  twistLimitDegrees=45
  twistMotion=Free
 }
 {
  bone1Name=dyn_jacket_fr_r_03
  bone2Name=dyn_jacket_side_r_03
  swing1LimitDegrees=45
  swing1Motion=Free
  swing2LimitDegrees=45
  swing2Motion=Free
  twistLimitDegrees=45
  twistMotion=Free
 }
 {
  bone1Name=dyn_jacket_side_r_03
  bone2Name=dyn_jacket_bk_r_02
  swing1LimitDegrees=45
  swing1Motion=Free
  swing2LimitDegrees=45
  swing2Motion=Free
  twistLimitDegrees=45
  twistMotion=Free
 }
]
