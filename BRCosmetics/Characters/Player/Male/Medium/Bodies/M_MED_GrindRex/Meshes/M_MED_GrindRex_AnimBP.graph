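# Post-process-style AnimBP for the GrindRex body cosmetic, adding secondary motion to clothing and accessories.
# The pose copied from the parent mesh runs through a Control Rig and jacket-strap RigidBody.
# A front-end flag selects between belt/jacket RigidBody chains, then AnimDynamics moves shorts, hood, zipper and necklace, and a constraint drives the zipper base.
# The event graph tracks frame rate, skydiving, crouching and gravity settings, and toggles tail visibility by backbling length.
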AnimBlueprint /BRCosmetics/Characters/Player/Male/Medium/Bodies/M_MED_GrindRex/Meshes/M_MED_GrindRex_AnimBP
target_skeleton=&/M_MED_GrindRex_Skeleton
graphs [
 AnimGraph {nodes [
   n1 = Anim Dynamics - Bone: dyn_shorts_l_frnt(ComponentPose=Local To Component(LocalPose=Use cached pose 'toDynamics'()), Alpha=1, ExternalForce="0,0,0") {
    angularDampingOverride=1
    angularSpringConstant=800
    bAngularSpring=true
    bOverrideAngularDamping=true
    bUseGravityOverride=true
    simulationSpace=RootRelative
    alphaBoolBlend {blendOption=Linear}
    boundBone {boneName=dyn_shorts_l_frnt}
    componentAppliedLinearAccClamp=<5000 5000 5000>
    componentLinearAccScale=<0.65 0.65 0.65>
    componentLinearVelScale=<0.65 0.65 0.65>
    gravityOverride {z=-980}
    physicsBodyDefinitions [{
      boundBone {boneName=dyn_shorts_l_frnt}
      boxExtents=<10 10 10>
      constraintSetup {coneAngle=1, angularLimitsMax=<1 20 15>, angularLimitsMin=<-1 -10 -15>, angularTarget=<1 0 0>}
      localJointOffset=<0.5 0 0>
     }]
    relativeSpaceBone {boneName=thigh_l}
   }
   n2 = Anim Dynamics - Bone: dyn_shorts_l_bk_01(ComponentPose=@n1, Alpha=1, ExternalForce="0,0,0") {
    angularDampingOverride=1
    angularSpringConstant=800
    bAngularSpring=true
    bOverrideAngularDamping=true
    bUseGravityOverride=true
    simulationSpace=RootRelative
    alphaBoolBlend {blendOption=Linear}
    boundBone {boneName=dyn_shorts_l_bk_01}
    componentAppliedLinearAccClamp=<10000 10000 10000>
    componentLinearAccScale=<1 1 1>
    componentLinearVelScale=<1 1 1>
    gravityOverride {z=-980}
    physicsBodyDefinitions [{
      boundBone {boneName=dyn_shorts_l_bk_01}
      boxExtents=<10 10 10>
      constraintSetup {angularLimitsMax=<1 2 10>, angularLimitsMin=<-1 -15 -10>, angularTarget=<1 0 0>}
      localJointOffset=<1 0 0>
     }]
   }
   n3 = Anim Dynamics - Bone: dyn_shorts_l_in(ComponentPose=@n2, Alpha=1, ExternalForce="0,0,0") {
    angularDampingOverride=1
    angularSpringConstant=800
    bAngularSpring=true
    bOverrideAngularDamping=true
    bUseGravityOverride=true
    simulationSpace=RootRelative
    alphaBoolBlend {blendOption=Linear}
    boundBone {boneName=dyn_shorts_l_in}
    componentAppliedLinearAccClamp=<10000 10000 10000>
    componentLinearAccScale=<1 1 1>
    componentLinearVelScale=<1 1 1>
    gravityOverride {z=-980}
    physicsBodyDefinitions [{
      boundBone {boneName=dyn_shorts_l_in}
      boxExtents=<10 10 10>
      constraintSetup {angularLimitsMax=<1 2 10>, angularLimitsMin=<-1 -15 -10>, angularTarget=<1 0 0>}
      localJointOffset=<1 0 0>
     }]
   }
   n4 = Anim Dynamics - Bone: dyn_shorts_l_out(ComponentPose=@n3, Alpha=1, ExternalForce="0,0,0") {
    angularDampingOverride=1
    angularSpringConstant=800
    bAngularSpring=true
    bOverrideAngularDamping=true
    bUseGravityOverride=true
    simulationSpace=RootRelative
    alphaBoolBlend {blendOption=Linear}
    boundBone {boneName=dyn_shorts_l_out}
    componentAppliedLinearAccClamp=<10000 10000 10000>
    componentLinearAccScale=<1 1 1>
    componentLinearVelScale=<1 1 1>
    gravityOverride {z=-980}
    physicsBodyDefinitions [{
      boundBone {boneName=dyn_shorts_l_out}
      boxExtents=<10 10 10>
      constraintSetup {angularLimitsMax=<1 2 10>, angularLimitsMin=<-1 -15 -10>, angularTarget=<1 0 0>}
      localJointOffset=<1 0 0>
     }]
   }
   n5 = Anim Dynamics - Bone: dyn_shorts_r_frnt(ComponentPose=@n4, Alpha=1, ExternalForce="0,0,0") {
    angularDampingOverride=1
    angularSpringConstant=800
    bAngularSpring=true
    bOverrideAngularDamping=true
    bUseGravityOverride=true
    simulationSpace=RootRelative
    alphaBoolBlend {blendOption=Linear}
    boundBone {boneName=dyn_shorts_r_frnt}
    componentAppliedLinearAccClamp=<5000 5000 5000>
    componentLinearAccScale=<0.65 0.65 0.65>
    componentLinearVelScale=<0.65 0.65 0.65>
    gravityOverride {z=-980}
    physicsBodyDefinitions [{
      boundBone {boneName=dyn_shorts_r_frnt}
      boxExtents=<10 10 10>
      constraintSetup {angularLimitsMax=<1 20 15>, angularLimitsMin=<-1 0 -15>, angularTarget=<1 0 0>}
      localJointOffset=<-0.5 0 0>
     }]
    relativeSpaceBone {boneName=root}
   }
   n6 = Anim Dynamics - Bone: dyn_shorts_r_bk_01(ComponentPose=@n5, Alpha=1, ExternalForce="0,0,0") {
    angularDampingOverride=1
    angularSpringConstant=800
    bAngularSpring=true
    bOverrideAngularDamping=true
    bUseGravityOverride=true
    simulationSpace=RootRelative
    alphaBoolBlend {blendOption=Linear}
    boundBone {boneName=dyn_shorts_r_bk_01}
    componentAppliedLinearAccClamp=<10000 10000 10000>
    componentLinearAccScale=<1 1 1>
    componentLinearVelScale=<1 1 1>
    gravityOverride {z=-980}
    physicsBodyDefinitions [{
      boundBone {boneName=dyn_shorts_r_bk_01}
      boxExtents=<10 10 10>
      constraintSetup {angularLimitsMax=<1 2 10>, angularLimitsMin=<-1 -15 -10>, angularTarget=<1 0 0>}
      localJointOffset=<-1 0 0>
     }]
   }
   n7 = Anim Dynamics - Bone: dyn_shorts_r_out(ComponentPose=@n6, Alpha=1, ExternalForce="0,0,0") {
    angularDampingOverride=1
    angularSpringConstant=800
    bAngularSpring=true
    bOverrideAngularDamping=true
    bUseGravityOverride=true
    simulationSpace=RootRelative
    alphaBoolBlend {blendOption=Linear}
    boundBone {boneName=dyn_shorts_r_out}
    componentAppliedLinearAccClamp=<10000 10000 10000>
    componentLinearAccScale=<1 1 1>
    componentLinearVelScale=<1 1 1>
    gravityOverride {z=-980}
    physicsBodyDefinitions [{
      boundBone {boneName=dyn_shorts_r_out}
      boxExtents=<10 10 10>
      constraintSetup {angularLimitsMax=<1 2 10>, angularLimitsMin=<-1 -15 -10>, angularTarget=<1 0 0>}
      localJointOffset=<-1 0.5 0>
     }]
   }
   n8 = Anim Dynamics - Bone: dyn_shorts_r_in(ComponentPose=@n7, Alpha=1, ExternalForce="0,0,0") {
    angularDampingOverride=1
    angularSpringConstant=800
    bAngularSpring=true
    bOverrideAngularDamping=true
    bUseGravityOverride=true
    simulationSpace=RootRelative
    alphaBoolBlend {blendOption=Linear}
    boundBone {boneName=dyn_shorts_r_in}
    componentAppliedLinearAccClamp=<10000 10000 10000>
    componentLinearAccScale=<1 1 1>
    componentLinearVelScale=<1 1 1>
    gravityOverride {z=-980}
    physicsBodyDefinitions [{
      boundBone {boneName=dyn_shorts_r_in}
      boxExtents=<10 10 10>
      constraintSetup {angularLimitsMax=<1 2 10>, angularLimitsMin=<-1 -15 -10>, angularTarget=<1 0 0>}
      localJointOffset=<-1 0 0>
     }]
   }
   n9 = Anim Dynamics - Bone: dyn_hood(ComponentPose=@n8, Alpha=1, ExternalForce="0,0,0") {
    angularDampingOverride=0.9999
    angularSpringConstant=300
    bAngularSpring=true
    bOverrideAngularDamping=true
    bUseGravityOverride=true
    simulationSpace=RootRelative
    alphaBoolBlend {blendOption=Linear}
    boundBone {boneName=dyn_hood}
    componentAppliedLinearAccClamp=<5000 5000 5000>
    componentLinearAccScale=<0.65 0.65 0.65>
    componentLinearVelScale=<0.65 0.65 0.65>
    gravityOverride {z=-980}
    physicsBodyDefinitions [{
      boundBone {boneName=dyn_hood}
      boxExtents=<25 25 25>
      constraintSetup {coneAngle=1, angularLimitsMax=<1 15 5>, angularLimitsMin=<-1 -15 -5>, angularTarget=<1 0 0>, linearAxesMax=<1 1 2>, linearAxesMin=<-1 -1 -2>}
      localJointOffset=<5 0 0>
     }]
    relativeSpaceBone {boneName=spine_05}
    simSpaceSettings {externalAngularVelocity {z=1}}
   }
   n10 = Anim Dynamics - Bone: dyn_jacket_zippo_01(ComponentPose=@n9, Alpha=1, ExternalForce="5,0,0") {
    angularSpringConstant=200
    bAngularSpring=true
    bUseGravityOverride=true
    simulationSpace=RootRelative
    boundBone {boneName=dyn_jacket_zippo_01}
    componentAppliedLinearAccClamp=<2000 2000 2000>
    componentLinearAccScale=<1 0.5 0.5>
    componentLinearVelScale=<1 0.5 0.5>
    gravityOverride {z=-980}
    physicsBodyDefinitions [{
      boundBone {boneName=dyn_jacket_zippo_01}
      boxExtents=<4 4 4>
      constraintSetup {angularLimitsMax=<0 10 45>, angularLimitsMin=<0 -10 -5>, angularTarget=<1 0 0>}
      localJointOffset=<0.1 0 0>
     }]
   }
   n11 = Anim Dynamics - Bone: dyn_necklace(ComponentPose=@n10, Alpha=1, ExternalForce="0,0,0") {
    angularSpringConstant=500
    bAngularSpring=true
    simulationSpace=RootRelative
    boundBone {boneName=dyn_necklace}
    componentAppliedLinearAccClamp=<3000 3000 3000>
    componentLinearAccScale=<1 0.7 0.7>
    componentLinearVelScale=<1 0.7 0.7>
    physicsBodyDefinitions [{
      boundBone {boneName=dyn_necklace}
      boxExtents=<6 6 6>
      constraintSetup {angularLimitsMax=<0 150 1>, angularLimitsMin=<0 0 -1>, angularTarget=<1 0 0>}
      localJointOffset=<0.1 0 0>
     }]
   }
   n12 = "Constraint - dyn_jacket_zippo_base (PPPPPP)"!(
    ComponentPose=@n11
    Alpha=1
    ConstraintWeights_0=0.0124
    ConstraintWeights_1:float=0.0277
    ConstraintWeights_2:float=0.9507
    ConstraintWeights_3:float=0.0023
    ConstraintWeights_4:float=0.0045
    ConstraintWeights_5:float=0.0024
   ) -> (Pose) {
    boneToModify {boneName=dyn_jacket_zippo_base}
    constraintSetup [
     {targetBone {boneName=dyn_jacket_fr_l_01}}
     {targetBone {boneName=dyn_jacket_fr_l_02}}
     {targetBone {boneName=dyn_jacket_fr_l_03}}
     {targetBone {boneName=dyn_jacket_side_l_01}}
     {targetBone {boneName=dyn_jacket_side_l_02}}
     {targetBone {boneName=dyn_jacket_side_l_03}}
    ]
    constraintWeights [1,1,1,1,1,1]
   }
   Output Pose(Result=Component To Local(ComponentPose=@n12)) {layerGroup=None}
   n13 = RigidBody(ComponentPose=Local To Component(LocalPose=Use cached pose 'toPhys'())) {
    bClampLinearTranslationLimitToRefPose=true
    bOverrideWorldGravity=true
    overridePhysicsAsset=&/M_MED_GrindRex_Physics_BeltStrapKeyChain
    simulationSpace=BaseBoneSpace
    baseBoneRef {boneName=pelvis}
    componentAppliedLinearAccClamp=<2500 2500 2500>
    overrideWorldGravity {z=-980}
    simSpaceSettings {
     dampingAlpha=0
     maxAngularAcceleration=32
     maxAngularVelocity=6
     maxLinearAcceleration=1200
     maxLinearVelocity=200
     velocityScaleZ=0.7
     worldAlpha=1
     externalAngularVelocity=<1 1 1>
     externalLinearDragV {x=1}
     externalLinearVelocity=<1 1 1>
    }
   }
   n14 = RigidBody(ComponentPose=@n13) {
    bClampLinearTranslationLimitToRefPose=true
    bOverrideWorldGravity=true
    overridePhysicsAsset=&/M_MED_GrindRex_Physics_Jacket_Frontend
    simulationSpace=BaseBoneSpace
    baseBoneRef {boneName=spine_02}
    overrideWorldGravity {z=-980}
    simSpaceSettings {
     dampingAlpha=0
     maxAngularAcceleration=32
     maxAngularVelocity=6
     maxLinearAcceleration=300
     maxLinearVelocity=50
     velocityScaleZ=0.7
     worldAlpha=1
     externalLinearDragV {x=1}
    }
   }
   n15 = RigidBody!(
    ComponentPose=Local To Component(LocalPose=Use cached pose 'toPhys'())
    Alpha
    OverrideWorldGravity:Vector="0,0,0"
    ExternalForce
    SimSpaceSettings:Sim Space Settings Structure=(WorldAlpha=0,VelocityScaleZ=1,DampingAlpha=1,MaxLinearVelocity=10000,MaxAngularVelocity=10000,MaxLinearAcceleration=10000,MaxAngularAcceleration=10000,ExternalLinearDragV=(X=0,Y=0,Z=0),ExternalLinearVelocity=(X=0,Y=0,Z=0),ExternalAngularVelocity=(X=0,Y=0,Z=0))
   ) -> (Pose) {
    bClampLinearTranslationLimitToRefPose=true
    bOverrideWorldGravity=true
    overridePhysicsAsset=&M_MED_GrindRex_Physics_BeltStrapKeyChain
    simulationSpace=BaseBoneSpace
    baseBoneRef {boneName=pelvis}
    componentAppliedLinearAccClamp=<3500 3500 3500>
   }
   n16 = RigidBody(
    ComponentPose=@n15
    OverrideWorldGravity="0,0,-980"
    SimSpaceSettings=(WorldAlpha=1,VelocityScaleZ=0.5,DampingAlpha=0,MaxLinearVelocity=100,MaxAngularVelocity=6,MaxLinearAcceleration=300,MaxAngularAcceleration=100,ExternalLinearDragV=(X=0,Y=0,Z=0),ExternalLinearVelocity=(X=0,Y=0,Z=0),ExternalAngularVelocity=(X=0,Y=0,Z=0))
   ) {
    bClampLinearTranslationLimitToRefPose=true
    bOverrideWorldGravity=true
    overridePhysicsAsset=&/M_MED_GrindRex_Physics_Jacket
    simulationSpace=BaseBoneSpace
    baseBoneRef {boneName=spine_02}
    overrideWorldGravity {z=-980}
    simSpaceSettings {
     dampingAlpha=0
     maxAngularAcceleration=100
     maxAngularVelocity=6
     maxLinearAcceleration=300
     maxLinearVelocity=100
     velocityScaleZ=0.5
     worldAlpha=1
    }
   }
   Save cached pose 'toDynamics'(Pose=Blend Poses by bool(BlendPose_0=Component To Local(ComponentPose=@n14), BlendPose_1=Component To Local(ComponentPose=@n16), bActiveValue=Get bIsInFrontEnd())) {cacheName=toDynamics}
   n17 = Control Rig!(
    Source=Copy Pose From Mesh() {bUseAttachedParent=true}
    Alpha
    Is Skydiving:bool=Get Is Skydiving() -> (Is Skydiving:bool)
    Is Crouching:bool=Get Is Crouching() -> (Is Crouching:bool)
    Low FPS:double=Get low fps() -> (low fps:double)
    SkyDaving:bool=Get bIsSkyDiving()
   ) -> (Pose) {controlRigAssetReference {blueprintRigClass=&/M_MED_GrindRex_CtrlRig.M_MED_GrindRex_CtrlRig_C}}
   n18 = RigidBody(
    ComponentPose=Local To Component(LocalPose=@n17)
    OverrideWorldGravity="0,0,0"
    SimSpaceSettings=(WorldAlpha=0,VelocityScaleZ=1,DampingAlpha=1,MaxLinearVelocity=10000,MaxAngularVelocity=10000,MaxLinearAcceleration=10000,MaxAngularAcceleration=10000,ExternalLinearDragV=(X=0,Y=0,Z=0),ExternalLinearVelocity=(X=0,Y=0,Z=0),ExternalAngularVelocity=(X=0,Y=0,Z=0))
   ) {
    bClampLinearTranslationLimitToRefPose=true
    bOverrideWorldGravity=true
    overridePhysicsAsset=&/M_MED_GrindRex_Physics_JacketStraps
    simulationSpace=BaseBoneSpace
    baseBoneRef {boneName=spine_05}
   }
   Save cached pose 'toPhys'(Pose=Component To Local(ComponentPose=@n18)) {cacheName=toPhys}
  ]}
 EventGraph {nodes [
   n1 = Event BlueprintUpdateAnimation()
   > n2 = Set Delta Time X(Delta Time X:double=@n1.DeltaTimeX) -> (Output_Get:double)
   > Set low fps(low fps:double=float / float(A=1, B=@n2)) -> (Output_Get:double)
   > n3 = Sequence()
   > n4 = "Gravity Override State Function (Parameter Asset) (Nativized)"(
    CharacterPartAnimInstance=Self-Reference()
    GameTimeInSeconds=GetGameTimeInSeconds()
    Parameters=&/M_MED_GrindRex_Parameters_var
   )
   > Set Gravity Override(Gravity Override=@n4)
   > Set RBAN Settings Out(RBAN Settings Out=@n4.OutRigidBodyAnimNodeParameters)
   n5 = Cast To PlayerPawn_Athena(execute=@n3.then_1, Object=TryGetPawnOwner())
   > Branch(Condition=CheckForTail!(self:Custom Character Part Anim Instance Object Reference) -> (bHasTail:bool))
   > n6 = Branch(Condition=CheckForBackbling().bIsBackblingLong)
   n7 = GetSkeletalMeshForPartType(self=@n5, PartType=MiscOrTail)
   > SetVisibility!(execute, self:Scene Component Object Reference=@n7, bNewVisibility:bool=false, bPropagateToChildren:bool=false) -> (then)
   SetVisibility(execute=@n6.else, self=@n7, bNewVisibility=true, bPropagateToChildren=false)
   Event OnBeginSkydiving()
   > Set Is Skydiving(Is Skydiving:bool=true) -> (Output_Get:bool)
   Event OnEndSkydiving()
   > Set Is Skydiving(Is Skydiving=false)
   Event OnBeginCrouching()
   > Set Is Crouching(Is Crouching:bool=1) -> (Output_Get:bool)
   Event OnEndCrouching()
   > Set Is Crouching(Is Crouching=false)
  ]}
]
variables [
 Delta Time X {type=number}
 Gravity Override {type=Vector}
 Is Crouching {type=boolean}
 Is Skydiving {type=boolean}
 low fps {type=number}
 RBAN Settings Out {
  type=FortRigidBodyAnimNodeParameters
  default {simSpaceSettings {dampingAlpha=0, maxAngularAcceleration=60, maxAngularVelocity=6, maxLinearAcceleration=600, maxLinearVelocity=200, velocityScaleZ=0.3}}
 }
]
